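; Prototype lazy-river tube: an invulnerable, buoyant physics BuildingActor that players deploy by interacting, knock around with hits and steer by standing on it.
; An overlap sphere starts a timer that, on water, applies impulses at the socket farthest from the rider, scaled by their offset from center.
; Buoyancy pontoon events and timers drive Niagara ripple effects.
; Much logic derives from the physics tree log: impact launches, damage effects and analytics.
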(event EventEndPlay
  (params (EndPlayReason :type EEndPlayReason%20Enum :default "Destroyed"))
  ;[1 1 1] End play
  (bind _returnvalue (BuildingActor|PhysicsSimulationSupport|GetPhysicsObjectComponent self))
  (bind (_numberofcachedplayer _onboardlocation _offboardlocation _sessionlength) (PhysicsObject|GetCachedRidingPlayersInfo _returnvalue true))
  (FortPhysicsAnalytics|FireTelemetryEventonPhysicsRidingActorLifeEnd self _numberofcachedplayer _onboardlocation _offboardlocation _sessionlength))

(event EventBeginPlay
  (bind _staticmeshcomponent (Variables|Debug|GetStaticMeshComponent))
  ;[1 1 1] Begin play
  (|CallParentFunction)
  (Variables|BuildingActor|SetbIsInvulnerable true)
  (Collision|AssignOnComponentHit (Variables|Debug|GetStaticMeshComponent) (EventDispatchers|CreateEvent :!event Custom|OnComponentHit_Event_0))
  (EventDispatchers|BindEventtoOnDamaged (EventDispatchers|CreateEvent :!event Custom|Hit) :!class /Script/FortniteGame.BuildingActor)
  (bind _output_get (Variables|Default|SetBuoyancyComponent (Variables|Default|GetBuoyancy)))
  (Cosmetic|BindEventtoOnEnteredWaterDelegate _output_get (EventDispatchers|CreateEvent :!event Custom|OnPontoonEnteredWater))
  (Cosmetic|BindEventtoOnExitedWaterDelegate _output_get (EventDispatchers|CreateEvent :!event Custom|OnPontoonExitedWater))
  (Cosmetic|BindEventtoOnEnteredWaterDelegate (Variables|Default|GetBuoyancy) (delegate Custom|InWater))
  (Cosmetic|BindEventtoOnExitedWaterDelegate (Variables|Default|GetBuoyancy) (delegate Custom|ExitedWater))
  (bind _returnvalue (BuildingActor|PhysicsSimulationSupport|GetPhysicsObjectComponent self))
  (PhysicsObject|BindEventtoOnAwakeChanged _returnvalue (EventDispatchers|CreateEvent :!event Custom|AwakeStateChanged))
  (Utilities|Time|SetTimerbyEvent (delegate Custom|VelocityCheck) 1.5 true)
  (bind _returnvalue_2 (Utilities|Time|SetTimerbyEvent (EventDispatchers|CreateEvent :!event Custom|Update%20Water%20FX) 0.45 true))
  (Variables|Default|SetWaterSurfaceInfoTimer _returnvalue_2)
  (bind _outactors (Actor|GetAllActorsOfClass "/Game/Athena/Apollo/Water/Blueprints/BP_FluidSim_FN.BP_FluidSim_FN_C"))
  (Variables|Default|SetFluidSim (Utilities|Array|Get(acopy) _outactors))
  (Niagara|SpawnSystemAttached "/Game/Environments/Apollo/Foliage/PhysicsTree/FX/NS_Physics_Tree_Section_Crack_Burst.NS_Physics_Tree_Section_Crack_Burst" _staticmeshcomponent "None" "0, 0,-768.000000" "0, 0, 0" "KeepRelativeOffset" true)
  (bind _returnvalue_3 (Niagara|SpawnSystemAttached "/Game/Environments/Apollo/Foliage/PhysicsTree/FX/NS_Physics_Tree_Section_Crack_Burst.NS_Physics_Tree_Section_Crack_Burst" _staticmeshcomponent "None" "0, 0,768.000000" "0, 0, 0" "KeepRelativeOffset" true))
  (Niagara|SetNiagaraVariable(Bool) _returnvalue_3 "PlayFallingWoodBits" false)
  (Niagara|SpawnSystemAttached "/Game/Environments/Apollo/Foliage/PhysicsTree/FX/NS_Physics_Tree_Falling_Splinters.NS_Physics_Tree_Falling_Splinters" _staticmeshcomponent "None" "0, 0, 0" "0, 0, 0" "KeepRelativeOffset" true)
  (bind _returnvalue_4 (Utilities|Time|SetTimerbyEvent (delegate Custom|RemoveSnow) 0.1 true))
  (Variables|Default|SetSnowRemovalTimerHandle _returnvalue_4))

(event Custom|InWater
  (params (Pontoon :type Spherical%20Pontoon%20Structure%20%28by%20ref%29))
  ;[1 1 1] Begin play
  (Variables|Default|SetIsInWater true))

(event Custom|ExitedWater
  (params (Pontoon :type Spherical%20Pontoon%20Structure%20%28by%20ref%29))
  ;[1 1 1] Begin play
  (Variables|Default|SetIsInWater false))

(event Custom|VelocityCheck
  ;[1 1 1] Begin play
  (Variables|Default|SetIsDonutMoving (> (Math|Vector|VectorLength (Math|Vector|VectorGetAbs (Transformation|GetVelocity self))) 5.0)))

(event Custom|RemoveSnow
  (bind _snowtimervalue (Variables|Default|GetSnowTimerValue))
  (bind _returnvalue (+ _snowtimervalue 0.05))
  (bind _staticmeshcomponent (Variables|Debug|GetStaticMeshComponent))
  ;[1 1 1] Begin play
  (if (<= _returnvalue 1.0)
    (Variables|Default|SetSnowTimerValue _returnvalue)
    (Rendering|Material|SetScalarParameterValueonMaterials _staticmeshcomponent "SnowHeightMask-Add" (Math|Float|Lerp -0.15 -15.0 (Math|Float|Power _returnvalue 3.0)))
    (Rendering|Material|SetScalarParameterValueonMaterials _staticmeshcomponent "SnowBlendBrightness" (Math|Float|Lerp 0.5 0.0 (Math|Float|Power _returnvalue 2.0)))
    (else
      (Utilities|Time|ClearandInvalidateTimerbyHandle (Variables|Default|GetSnowRemovalTimerHandle)))))

(event Custom|OnComponentHit_Event_0
  (params (HitComponent :type Primitive%20Component%20Object%20Reference) (OtherActor :type Actor%20Object%20Reference) (OtherComp :type Primitive%20Component%20Object%20Reference) (NormalImpulse :type Vector) (Hit :type Hit%20Result%20Structure%20%28by%20ref%29))
  (bind _returnvalue (Transformation|GetVelocity))
  (bind _returnvalue_2 (Math|Vector|VectorLength _returnvalue))
  (bind (_bblockinghit _binitialoverlap _time _distance _location _impactpoint _normal _impactnormal) (Collision|BreakHitResult Hit))
  (bind (_minimum_met _impulse) (CallFunction|GetPlayerImpulseVelocityfromImpactVelocity _returnvalue_2))
  (bind _returnvalue_3 (Utilities|Time|GetTimeSeconds))
  (bind _nextpotentiallaunchtime (Variables|Default|GetNextPotentialLaunchTime))
  (bind _returnvalue_4 (> _returnvalue_3 _nextpotentiallaunchtime))
  (bind _ns_physics_tree_environment_impact (Variables|Default|GetNS_Physics_Tree_Environment_Impact))
  ;[1 1 1] Log hit (Bound when active)
  (if (Networking|IsServer)
    (Variables|Default|SetLastImpactNormal _impactnormal)
    (bind _returnvalue_5 (BuildingActor|PhysicsSimulationSupport|GetPhysicsObjectComponent self))
    (if (Components|Activation|IsActive _returnvalue_5)
      ;[1 1 1] Launch player
      (bind _asfort_player_pawn (Utilities|Casting|CastToFortPlayerPawn OtherActor))
      (if (== (Variables|Debug|GetStaticMeshComponent) (Pawn|Components|CharacterMovement|GetMovementBaseObject (Variables|/Script/Engine.Character|GetCharacterMovement _asfort_player_pawn)))
        ;;[1 1 1] Set a launch delay for certain interactions, like riding the log in water. This will prevent the player from accidentally being launched when performing cool water tricks and jumps
        (if _returnvalue_4
          (Variables|Default|SetNextPotentialLaunchTime (+ (Utilities|Time|GetTimeSeconds) (Variables|Default|GetLaunchDelay))))
        (elif (Buoyancy|IsinWaterBody (Variables|Default|GetBuoyancyComponent))
          _
          (elif _minimum_met
            (if _returnvalue_4
              (Character|LaunchCharacter _asfort_player_pawn (* (Math|Vector|Normalize (+ (* (Variables|Default|GetLastImpactNormal) "-1.000000,-1.000000,0.000000") (Math|Vector|MakeVector 0.0 0.0 (ScalableFloat|GetValueatLevel (Variables|Default|GetVerticalImpulseRatio))))) _impulse) true true))))))
    (elif (!= (Utilities|GetClass OtherActor) (Utilities|GetClass (Game|GetPlayerPawn 0)))
      (Transformation|SetWorldTransform (Variables|Default|GetNS_Physics_Tree_Environment_Impact) (Math|Transform|MakeTransform _location (Math|Conversions|RotationFromXVector _impactnormal)) false false)
      (if (not (Buoyancy|IsinWaterBody (Variables|Default|GetBuoyancyComponent)))
        (bind _returnvalue_6 (BuildingActor|PhysicsSimulationSupport|GetPhysicsObjectComponent self))
        (if (not (<= (Math|Vector|VectorLength (Variables|/Script/FortniteGame.FortPhysicsObjectComponent|GetLinearVelocity _returnvalue_6)) 8.0))
          (Niagara|SetNiagaraVariable(Float) _ns_physics_tree_environment_impact "Scale" (Math|Float|Clamp(Float) (* (Math|Vector|VectorLength NormalImpulse) 2e-06) 0.45))
          (Components|Activation|Activate _ns_physics_tree_environment_impact false)))
      (Components|Activation|Deactivate _ns_physics_tree_environment_impact))))

(event Custom|LogDamaged
  (params (DamagedActor :type Actor%20Object%20Reference) (Damage :type Float%20%28single-precision%29) (InstigatedBy :type Controller%20Object%20Reference) (DamageCauser :type Actor%20Object%20Reference) (HitLocation :type Vector) (FHitComponent :type Primitive%20Component%20Object%20Reference) (BoneName :type Name) (Momentum :type Vector))
  ;[1 1 1] Log damaged
  (Utilities|FlowControl|Sequence
    (:then_0
      (bind _asfort_player_pawn (Utilities|Casting|CastToFortPlayerPawn DamageCauser))
      (FortPhysicsSimulationLibrary|SetImpactInstigatorforPhysicsObject self _asfort_player_pawn))
    (:then_1
      (Variables|Default|SetBreakTreeDamage (- 1.0 (/ (Health|GetHealth) (Health|GetMaxHealth)))))))

(event Custom|Update%20Water%20FX
  ;[1 1 1] Update Water Body Surface Info to Niagara System
  (if (Buoyancy|IsOverlappingWaterBody (Variables|Default|GetBuoyancyComponent))
    (bind (_outwaterplanelocation _outwaterplanenormal _outwatersurfaceposition _outwaterdepth _outwaterbodyidx) (Buoyancy|GetLastWaterSurfaceInfo (Variables|Default|GetBuoyancyComponent)))
    (Variables|FX|SetWaterFxPlaneLocation _outwatersurfaceposition)
    (Variables|FX|SetWaterFxPlaneNormal _outwaterplanenormal)
    (Variables|FX|SetWaterFxPlaneDepth _outwaterdepth)
    (Variables|FX|SetWaterFxIndex _outwaterbodyidx)
    (CallFunction|AdjustWaterSettings)
    (else
      (CallFunction|AdjustWaterSettings)
      (Utilities|Time|PauseTimerbyHandle (Variables|Default|GetWaterSurfaceInfoTimer)))))

(event Custom|CE_SplashInWater
  (params (SplashLocation :type Vector))
  ;[1 1 1] Main Water Splash - Still needs Work
  (bind _returnvalue (BuildingActor|PhysicsSimulationSupport|GetPhysicsObjectComponent self))
  (if (>= (Math|Vector|VectorLength (Variables|/Script/FortniteGame.FortPhysicsObjectComponent|GetLinearVelocity _returnvalue)) 7.8e+02)
    (bind _returnvalue_2 (Niagara|SpawnSystematLocation "/Game/Environments/Apollo/Foliage/PhysicsTree/FX/NS_Physics_Tree_Splash.NS_Physics_Tree_Splash" SplashLocation))
    (Niagara|SetNiagaraVariable(Float) _returnvalue_2 "EffectScale" (Math|Float|Clamp(Float) (* (Math|Vector|VectorLength (Variables|/Script/FortniteGame.FortPhysicsObjectComponent|GetLinearVelocity _returnvalue)) 0.0025) 0.55))
    (Niagara|SetNiagaraVariable(Float) _returnvalue_2 "WaterDepth" (Variables|FX|GetWaterFxPlaneDepth))
    (Niagara|SetNiagaraVariable(Vector3) _returnvalue_2 "waterPlaneLocation" (Variables|FX|GetWaterFxPlaneLocation))
    (Niagara|SetNiagaraVariable(Vector3) _returnvalue_2 "waterPlaneNormal" (Variables|FX|GetWaterFxPlaneNormal))
    (Niagara|SetNiagaraVariable(Float) _returnvalue_2 "WaterBodyIndex" (Math|Conversions|ToFloat(Integer) (Variables|FX|GetWaterFxIndex)))))

(event EventOnDeathPlayEffects
  (params (Damage :type Float%20%28single-precision%29) (DamageTags :type Gameplay%20Tag%20Container%20Structure%20%28by%20ref%29) (Momentum :type Vector :default "0, 0, 0") (HitInfo :type Hit%20Result%20Structure%20%28by%20ref%29) (InstigatedBy :type Fort%20Pawn%20Object%20Reference) (DamageCauser :type Actor%20Object%20Reference) (EffectContext :type Gameplay%20Effect%20Context%20Handle%20Structure)))

(event EventOnDeathServer
  (params (Damage :type Float%20%28single-precision%29) (DamageTags :type Gameplay%20Tag%20Container%20Structure%20%28by%20ref%29) (Momentum :type Vector :default "0, 0, 0") (HitInfo :type Hit%20Result%20Structure%20%28by%20ref%29) (InstigatedBy :type Controller%20Object%20Reference) (DamageCauser :type Actor%20Object%20Reference) (EffectContext :type Gameplay%20Effect%20Context%20Handle%20Structure))
  ;[1 1 1] Death Analytics
  (bind (_returnvalue _success) (CallFunction|GetAnalyticsReporterActor))
  (if _success
    (Class|BPhysicsTreeReporter|ReportLogDestroyed _returnvalue (Utilities|GetObjectName self) (Transformation|GetActorLocation self))))

(event Custom|AdjustWaterSettings
  (Networking|CanExecuteCosmeticEvents
    (:True
      (for _array_element (Variables|Default|GetFxSystemArray)
        (if (Buoyancy|IsOverlappingWaterBody (Variables|Default|GetBuoyancyComponent))
          (Components|Activation|Activate _array_element false)
          (Utilities|Time|UnpauseTimerbyHandle (Variables|Default|GetWaterSurfaceInfoTimer))
          (Effects|Components|ParticleSystem|SetVectorParameter _array_element "WaterPlaneLocation" (Variables|FX|GetWaterFxPlaneLocation))
          (Effects|Components|ParticleSystem|SetFloatParameter _array_element "WaterDepth" (Variables|FX|GetWaterFxPlaneDepth))
          (Effects|Components|ParticleSystem|SetFloatParameter _array_element "WaterBodyIndex" (Math|Conversions|ToFloat(Integer) (Variables|FX|GetWaterFxIndex)))
          (Effects|Components|ParticleSystem|SetVectorParameter _array_element "WaterPlaneNormal" (Variables|FX|GetWaterFxPlaneNormal))
          (else
            (Components|Activation|Deactivate _array_element)))))))

(event Custom|Control2dSimForces
  (bind _fluidforcedynamic (Utilities|Struct|MakeFluidForceDynamic "NewEnumerator0" 9e+02 1.5 0 "/Game/Environments/Apollo/Foliage/PhysicsTree/FX/Materials/M_FluidForce_Capsule3D_Inst.M_FluidForce_Capsule3D_Inst"))
  (bind _fluidsim (Variables|Default|GetFluidSim))
  ;[1 1 1] Register 2D Sim Force
  (Utilities|IsValid _fluidsim
    (:Is%20Valid
      (Variables|FX|SetFluid%20Force%20Dynamic _fluidforcedynamic)
      (Class|BPFluidSim01|RegisterProjectileForce _fluidsim _fluidforcedynamic 0))))

(event Custom|OnPontoonEnteredWater
  (params (Pontoon_CenterSocket :type Name) (Pontoon_RelativeLocation :type Vector) (Pontoon_Radius :type Float%20%28single-precision%29) (Pontoon_bFXEnabled :type Boolean) (Pontoon_LocalForce :type Vector) (Pontoon_CenterLocation :type Vector) (Pontoon_SocketRotation :type Quat%20Structure) (Pontoon_Offset :type Vector) (Pontoon_WaterHeight :type Float%20%28single-precision%29) (Pontoon_WaterDepth :type Float%20%28single-precision%29) (Pontoon_ImmersionDepth :type Float%20%28single-precision%29) (Pontoon_WaterPlaneLocation :type Vector) (Pontoon_WaterPlaneNormal :type Vector) (Pontoon_WaterSurfacePosition :type Vector) (Pontoon_WaterVelocity :type Vector) (Pontoon_WaterBodyIndex :type Integer) (Pontoon_bIsInWater :type Boolean) (Pontoon_CurrentWaterBodyComponent :type Water%20Body%20Component%20Object%20Reference))
  ;[1 1 1] Overlap Events for Water Effects
  (Networking|CanExecuteCosmeticEvents
    (:True
      (switch name Pontoon_CenterSocket
        (:Top
          (Utilities|IsValid (Variables|FX|GetWaterFxPhysicsTreeTop)
            (:Is%20Valid
              (CallFunction|AdjustWaterSettings))
            (:Is%20Not%20Valid
              (bind _returnvalue (Effects|SpawnFXSystemAttached :FXSystemAsset "/Game/Environments/Apollo/Foliage/PhysicsTree/FX/NS_Physics_Tree_InWater_Ripples.NS_Physics_Tree_InWater_Ripples" :AttachPointName "None" :Scale "1.000000,1.000000,1.000000" :LocationType "SnapToTarget" :bAutoDestroy true :PoolingMethod "None" :bAutoActivate true))
              (bind _output_get (Variables|FX|SetWaterFxPhysicsTreeTop _returnvalue))
              (Utilities|Array|Add (Variables|Default|GetFxSystemArray) _output_get))))
        (:Middle
          (Utilities|IsValid (Variables|FX|GetWaterFxPhysicsTreeMiddle)
            (:Is%20Valid
              (CallFunction|AdjustWaterSettings))
            (:Is%20Not%20Valid
              (bind _returnvalue_2 (Effects|SpawnFXSystemAttached :FXSystemAsset "/Game/Environments/Apollo/Foliage/PhysicsTree/FX/NS_Physics_Tree_InWater_Ripples.NS_Physics_Tree_InWater_Ripples" :AttachPointName "None" :Scale "1.000000,1.000000,1.000000" :LocationType "SnapToTarget" :bAutoDestroy true :PoolingMethod "None" :bAutoActivate true))
              (bind _output_get_2 (Variables|FX|SetWaterFxPhysicsTreeMiddle _returnvalue_2))
              (Utilities|Array|Add (Variables|Default|GetFxSystemArray) _output_get_2))))
        (:Bottom
          (Utilities|IsValid (Variables|FX|GetWaterFxPhysicsTreeBottom)
            (:Is%20Valid
              (CallFunction|AdjustWaterSettings))
            (:Is%20Not%20Valid
              (bind _returnvalue_3 (Effects|SpawnFXSystemAttached :FXSystemAsset "/Game/Environments/Apollo/Foliage/PhysicsTree/FX/NS_Physics_Tree_InWater_Ripples.NS_Physics_Tree_InWater_Ripples" :AttachPointName "None" :Scale "1.000000,1.000000,1.000000" :LocationType "SnapToTarget" :bAutoDestroy true :PoolingMethod "None" :bAutoActivate true))
              (bind _output_get_3 (Variables|FX|SetWaterFxPhysicsTreeBottom _returnvalue_3))
              (Utilities|Array|Add (Variables|Default|GetFxSystemArray) _output_get_3))))))))

(event OnComponentBeginOverlap(OverlapSphere)
  (params (OverlappedComponent :type Primitive%20Component%20Object%20Reference) (OtherActor :type Actor%20Object%20Reference) (OtherComp :type Primitive%20Component%20Object%20Reference) (OtherBodyIndex :type Integer) (bFromSweep :type Boolean) (SweepResult :type Hit%20Result%20Structure%20%28by%20ref%29))
  ; Steering
  (bind _asplayer_pawn_athena (Utilities|Casting|CastToPlayerPawn_Athena OtherActor))
  (Variables|Default|SetCurrentPawn _asplayer_pawn_athena)
  (bind _returnvalue (Utilities|Time|SetTimerbyEvent (delegate Custom|SteerCraft) 0.5 true))
  (Variables|Default|SetTimerHandle _returnvalue))

(event Custom|ReInitCheck
  ; Steering
  (bind _returnvalue (Utilities|Time|SetTimerbyEvent (delegate Custom|SteerCraft) 0.5 true))
  (Variables|Default|SetTimerHandle _returnvalue))

(event Custom|SteerCraft
  ; Steering
  (if (Buoyancy|IsOverlappingWaterBody (Variables|Default|GetBuoyancyComponent))
    (if (Utilities|IsValid (Variables|Default|GetCurrentPawn) :!class /Script/Engine.KismetSystemLibrary)
      (bind _isondonut (CallFunction|IsPlayeronDonut))
      (if _isondonut
        (bind _alldistancesockets (CallFunction|GetDistanceSockets))
        (bind (_distantvector _succeeded) (CallFunction|WhichDistanceSocketIsFurthestAway _alldistancesockets))
        (if _succeeded
          (bind (_magnitudenormalized _vector) (CallFunction|GetImpulseandVectorMagnitude _distantvector))
          (FortPhysicsSimulationLibrary|AddPhysicsImpulseatLocation self (* _vector (* _magnitudenormalized 1e+02)) _distantvector false "None" true 0 "(MaxSimSizeAllowed=Large,bRejectEmbeddedInLandscape=True)"))
        (else
          _))))
  (Utilities|Time|PauseTimerbyHandle (Variables|Default|GetTimerHandle))
  (Variables|Default|SetCurrentPawn 0)
  (Utilities|FlowControl|Delay (Variables|Default|GetRetestPlayerOverlapOnExitDelay))
  (CallFunction|RetestPlayerOverlap))

(event Custom|AwakeStateChanged
  (params (SimulatingComponent :type Primitive%20Component%20Object%20Reference) (bIsAwake :type Boolean))
  (bind _fxsystemarray (Variables|Default|GetFxSystemArray))
  ;[1 1 1] When Log Awake State Changes, Activate or Deactivate Water Particles
  (if bIsAwake
    (for _array_element _fxsystemarray
      (Components|Activation|Activate _array_element false))
    (else
      (for _array_element_2 _fxsystemarray
        (Components|Activation|Deactivate _array_element_2)))))

(event Custom|RetestPlayerOverlap
  ; Steering
  (bind _outactors (Collision|SphereOverlapActors (Transformation|GetWorldLocation (Variables|Default|GetOverlapSphere)) (Variables|/Script/Engine.SphereComponent|GetSphereRadius (Variables|Default|GetOverlapSphere)) (Utilities|Array|MakeArray "ObjectTypeQuery3") "/Script/FortniteGame.FortPlayerPawnAthena"))
  (if (> (Utilities|Array|Length _outactors) 0)
    (bind _asplayer_pawn_athena (Utilities|Casting|CastToPlayerPawn_Athena (Utilities|Array|Get(acopy) _outactors)))
    (Variables|Default|SetCurrentPawn _asplayer_pawn_athena)
    (CallFunction|ReInitCheck)))

(event Custom|OnPontoonExitedWater
  (params (Pontoon_CenterSocket :type Name) (Pontoon_RelativeLocation :type Vector) (Pontoon_Radius :type Float%20%28single-precision%29) (Pontoon_bFXEnabled :type Boolean) (Pontoon_LocalForce :type Vector) (Pontoon_CenterLocation :type Vector) (Pontoon_SocketRotation :type Quat%20Structure) (Pontoon_Offset :type Vector) (Pontoon_WaterHeight :type Float%20%28single-precision%29) (Pontoon_WaterDepth :type Float%20%28single-precision%29) (Pontoon_ImmersionDepth :type Float%20%28single-precision%29) (Pontoon_WaterPlaneLocation :type Vector) (Pontoon_WaterPlaneNormal :type Vector) (Pontoon_WaterSurfacePosition :type Vector) (Pontoon_WaterVelocity :type Vector) (Pontoon_WaterBodyIndex :type Integer) (Pontoon_bIsInWater :type Boolean) (Pontoon_CurrentWaterBodyComponent :type Water%20Body%20Component%20Object%20Reference)))

(event OnComponentEndOverlap(OverlapSphere)
  (params (OverlappedComponent :type Primitive%20Component%20Object%20Reference) (OtherActor :type Actor%20Object%20Reference) (OtherComp :type Primitive%20Component%20Object%20Reference) (OtherBodyIndex :type Integer))
  ; Disable when no player onboard
  (Utilities|Casting|CastToPlayerPawn_Athena OtherActor)
  (bind _isondonut (CallFunction|IsPlayeronDonut))
  (if (not _isondonut)
    (Utilities|Time|PauseTimerbyHandle (Variables|Default|GetTimerHandle))
    (Variables|Default|SetCurrentPawn 0)
    (Utilities|FlowControl|Delay 2.5)
    (CallFunction|RetestPlayerOverlap)))

(event Interaction|EventBlueprintOnInteract
  (params (InteractingPawn :type Fort%20Pawn%20Object%20Reference) (InteractionBeingAttempted :type EInteractionBeingAttempted%20Enum :default "FirstInteraction"))
  ; Deploy
  (Variables|Default|SetHasBeenDeployed true)
  (FortPhysicsSimulationLibrary|AddPhysicsImpulseatLocation self (+ (* (Transformation|GetActorForwardVector InteractingPawn) (Math|Conversions|ToVector(Float) 5.5e+02)) "0.000000,0.000000,425.000000") (Transformation|GetActorLocation self) false "None" true 0 "(MaxSimSizeAllowed=Large,bRejectEmbeddedInLandscape=True)")
  (Utilities|FlowControl|Delay 2.5)
  (if (not (Buoyancy|IsOverlappingWaterBody (Variables|Default|GetBuoyancyComponent)))
    (Variables|Default|SetHasBeenDeployed false)))

(event Custom|Hit
  (params (DamagedActor :type Actor%20Object%20Reference) (Damage :type Float%20%28single-precision%29) (InstigatedBy :type Controller%20Object%20Reference) (DamageCauser :type Actor%20Object%20Reference) (HitLocation :type Vector) (FHitComponent :type Primitive%20Component%20Object%20Reference) (BoneName :type Name) (Momentum :type Vector))
  ; Deploy
  (Utilities|FlowControl|DoOnce false :!label _label
    (:Completed
      (Utilities|FlowControl|Sequence
        (:then_0
          (if (Variables|Default|GetIsInWater)
            (Variables|Default|SetHitImpulseCurrentMultiplier 0.6)
            (else
              (Variables|Default|SetHitImpulseCurrentMultiplier 1.0)))
          (bind _asbga_physics_test_tube_1 (Utilities|Casting|CastToBGA_PhysicsTestTube1 DamagedActor))
          (FortPhysicsSimulationLibrary|AddPhysicsImpulseatLocation _asbga_physics_test_tube_1 (* (+ (* Momentum (Math|Conversions|ToVector(Float) 6.5e+02)) "0.000000,0.000000,800.000000") (Variables|Default|GetHitImpulseCurrentMultiplier)) (Building|GetCentroid self) false "None" true 0 "(MaxSimSizeAllowed=Large,bRejectEmbeddedInLandscape=True)"))
        (:then_1
          (Utilities|FlowControl|Delay 0.77)
          (goto _label Reset))))) ; stop players from flying in the sky :D 
)

(fn BlueprintCanInteract
  (params (InteractingPawn :type Fort%20Pawn%20Object%20Reference) (InteractionBeingAttempted :type EInteractionBeingAttempted%20Enum :default "FirstInteraction") (InteractionType :type TInteractionType%20Enum :default "IT_NoInteraction"))
  (if (Variables|Default|GetIsDonutMoving)
    (return (!= (!= (Buoyancy|IsOverlappingWaterBody (Variables|Default|GetBuoyancy)) false) (Variables|Default|GetIsInWater)))
    (else
      (return true))))

(fn BlueprintGetInteractionString
  (params (InteractingPawn :type Fort%20Pawn%20Object%20Reference) (InteractionBeingAttempted :type EInteractionBeingAttempted%20Enum :default "FirstInteraction"))
  (return "NSLOCTEXT(\"[6CA0C6BAF4B555EA3B6A71F566CED76C]\", \"458CD75D41F263AFE58ECB8B6218D411\", \"Deploy Tube\")"))

(fn BlueprintGetInteractionTime
  (params (InteractingPawn :type Fort%20Pawn%20Object%20Reference) (InteractionBeingAttempted :type EInteractionBeingAttempted%20Enum :default "FirstInteraction"))
  (return 0.5 true))

(fn CalculateSplitOffset
  (params (bSmallSide :type Boolean) (SplitPercentage :type Float%20%28double-precision%29))
  (if bSmallSide
    (return (* (- SplitPercentage (/ SplitPercentage 2.0)) (Variables|Default|GetLogLength)))
    (else
      (return (* (/ (- (Variables|Default|GetLogLength) (* (Variables|Default|GetLogLength) SplitPercentage)) 2.0) -1.0)))))

(fn CalculateSplitScale
  (params (bSmallSide :type Boolean) (SplitPercentage :type Float%20%28double-precision%29))
  (bind _staticmeshcomponent (Variables|Debug|GetStaticMeshComponent))
  (bind (_returnvalue_location _returnvalue_rotation _returnvalue_scale) (Math|Transform|BreakTransform (Transformation|GetWorldTransform _staticmeshcomponent)))
  (bind (_returnvalue_scale_x _returnvalue_scale_y _returnvalue_scale_z) (Math|Vector|BreakVector _returnvalue_scale))
  (if bSmallSide
    (return (* _returnvalue_scale_z (- 1.0 SplitPercentage)))
    (else
      (return (* _returnvalue_scale_z SplitPercentage)))))

(fn GetAnalyticsReporterActor
  (bind _returnvalue (Actor|GetActorOfClass "/Game/Athena/Items/PhysicsActors/PhysicsTreeLog/B_PhysicsTree_Reporter.B_PhysicsTree_Reporter_C"))
  (Utilities|IsValid _returnvalue
    (:Is%20Valid
      (return _returnvalue true))
    (:Is%20Not%20Valid
      (if (Networking|IsServer)
        (bind _returnvalue_2 (Game|SpawnActorfromClass "/Game/Athena/Items/PhysicsActors/PhysicsTreeLog/B_PhysicsTree_Reporter.B_PhysicsTree_Reporter_C" (Math|Transform|MakeTransform "0, 0, 0") "AlwaysSpawn" "OverrideRootScale" 0 0))
        (return _returnvalue_2 true)
        (else
          (return 0 false))))) ; Already have a valid reporter actor?
)

(fn GetDistanceSockets
  (for _array_element (Transformation|GetAllSocketNames (Variables|Debug|GetStaticMeshComponent))
    (if (not (== _array_element "Center"))
      (Utilities|Array|Add (LocalVariables|GetDistanceSocketsOnlyLocal) _array_element)))
  (return (LocalVariables|GetDistanceSocketsOnlyLocal)))

(fn GetImpulseAndVectorMagnitude
  (params (VectorIn :type Vector))
  (return (Math|Float|NormalizetoRange (Math|Float|Clamp(Float) (Math|Float|Absolute(Float) (Math|Vector|VectorLength (- (Transformation|GetSocketLocation (Variables|Debug|GetStaticMeshComponent) "Center") (Transformation|GetActorLocation (Variables|Default|GetCurrentPawn))))) 1.3e+02 2e+02) 1.3e+02 2e+02) (Math|Vector|GetUnitDirection(Vector) VectorIn (Transformation|GetActorLocation (Variables|Default|GetCurrentPawn)))))

(fn GetPlayerImpulseVelocityFromImpactVelocity
  (params (Impact :type Float%20%28double-precision%29))
  (bind (_arrayindex _arrayelement) (Utilities|Array|ReverseforEachLoop (Variables|Default|GetPlayerImpactTiers))
    (:LoopBody
      (if (>= (Variables|GetImpact) (ScalableFloat|GetValueatLevel _arrayelement))
        (LocalVariables|SetReturnImpulseVelocity (ScalableFloat|GetValueatLevel (Utilities|Array|Get(acopy) (Variables|Default|GetPlayerImpulseTiers) _arrayindex)))))
    (:Completed
      (return (>= (LocalVariables|GetReturnImpulseVelocity) 0) (LocalVariables|GetReturnImpulseVelocity)))))

(fn IsPlayerOnDonut
  (bind _outactors (Collision|SphereOverlapActors (Transformation|GetWorldLocation (Variables|Default|GetOverlapSphere)) (Variables|/Script/Engine.SphereComponent|GetSphereRadius (Variables|Default|GetOverlapSphere)) (Utilities|Array|MakeArray "ObjectTypeQuery3") "/Script/FortniteGame.FortPlayerPawnAthena"))
  (return (>= (Utilities|Array|Length _outactors) 1)))

(fn OnRep_BreakTreeDamage
  (bind _damagefxsocketdatatable (Variables|Default|GetDamageFXSocketDataTable))
  ;[1 1 1] NewFXLogic Yi.XIe 11.2
  (Utilities|FlowControl|Sequence
    (:then_0
      ;;[1 1 1] Get all socket information
      (if (not (Variables|Default|GetDamageFXSocketListInitialed))
        (bind _outrownames (DataTable|GetDataTableRowNames _damagefxsocketdatatable))
        (bind _output_get (Variables|Default|SetDamageFXSocketNameList _outrownames))
        (for _array_element _output_get
          (bind _returnvalue (Utilities|GetDataTableRow _damagefxsocketdatatable _array_element))
          (Utilities|Array|Add (Variables|Default|GetDamageFXSocketValueList) (.Cdr_13_7CF3136F40A0540EB935CEB549C90884 (Utilities|Struct|BreakPhysicsTreeFXSocketPos _returnvalue))))
        (Variables|Default|SetDamageFXSocketListInitialed true)))
    (:then_1
      (Utilities|FlowControl|Sequence
        (:then_0
          ;;[1 1 1] Remap damage value to get better visual
          (LocalVariables|SetRemapedDamage (Math|Curves|GetFloatValue (Variables|Default|GetDamageRemapCurve) (Variables|Default|GetBreakTreeDamage))))
        (:then_1
          ;;[1 1 1] Set material parameter
          (Rendering|Material|SetScalarParameterValueonMaterials (Variables|Debug|GetStaticMeshComponent) "BreakTreeDamage" (LocalVariables|GetRemapedDamage)))
        (:then_2
          ;;[1 1 1] Find current execution latest break socket index
          (LocalVariables|SetThisFrameDamageIndex -1)
          (bind (_array_element_2 _array_index) (Utilities|Array|ForEachLoopwithBreak (Variables|Default|GetDamageFXSocketValueList) :!label _label)
            (:LoopBody
              (if (> _array_element_2 (LocalVariables|GetRemapedDamage))
                (goto _label Break)
                (else
                  (LocalVariables|SetThisFrameDamageIndex _array_index))))))
        (:then_3
          ;;[1 1 1] Play FX from previous index to current index
          (if (not (or (<= (LocalVariables|GetThisFrameDamageIndex) (Variables|Default|GetCachedDamageValue)) (>= (LocalVariables|GetRemapedDamage) 1.0)))
            (for _index (range (Variables|Default|GetCachedDamageValue) (+ (LocalVariables|GetThisFrameDamageIndex) 1))
              (bind _returnvalue_2 (Utilities|GetDataTableRow (Variables|Default|GetDamageFXSocketDataTable) (Utilities|Array|Get(aref) (Variables|Default|GetDamageFXSocketNameList) _index)))
              (bind _returnvalue_3 (Effects|Components|ParticleSystem|SpawnFXAtLocation "/Game/Environments/Apollo/Foliage/PhysicsTree/FX/NS_Physics_Tree_Section_Burst.NS_Physics_Tree_Section_Burst" (Math|Conversions|ToVector(Vector4) (Math|Matrix|TransformPosition(Matrix) (Math|Transform|ToMatrix(Transform) (Transformation|GetWorldTransform (Variables|Debug|GetStaticMeshComponent))) (Math|Vector|MakeVector (Utilities|Struct|BreakPhysicsTreeFXSocketPos _returnvalue_2) (.Y_7_C54973424DE6B013DD840F8154423A31 (Utilities|Struct|BreakPhysicsTreeFXSocketPos _returnvalue_2)) (.Z_9_8C9F856849E0166DF28585899F82A337 (Utilities|Struct|BreakPhysicsTreeFXSocketPos _returnvalue_2))))) (Math|Conversions|RotationFromXVector (Math|Vector|ProjectVectorontoPlane (Math|Vector|Normalize (- (Math|Conversions|ToVector(Vector4) (Math|Matrix|TransformPosition(Matrix) (Math|Transform|ToMatrix(Transform) (Transformation|GetWorldTransform (Variables|Debug|GetStaticMeshComponent))) (Math|Vector|MakeVector (Utilities|Struct|BreakPhysicsTreeFXSocketPos _returnvalue_2) (.Y_7_C54973424DE6B013DD840F8154423A31 (Utilities|Struct|BreakPhysicsTreeFXSocketPos _returnvalue_2)) (.Z_9_8C9F856849E0166DF28585899F82A337 (Utilities|Struct|BreakPhysicsTreeFXSocketPos _returnvalue_2))))) (Transformation|GetWorldLocation (Variables|Debug|GetStaticMeshComponent)))) (Transformation|GetUpVector (Variables|Debug|GetStaticMeshComponent))))))
              (Effects|Components|ParticleSystem|SetFloatParameter _returnvalue_3 "BurstScale" (LocalVariables|GetRemapedDamage)))))
        (:then_4
          ;;[1 1 1] Cache
          (Variables|Default|SetCachedDamageValue (LocalVariables|GetThisFrameDamageIndex)))))))

(fn ReportTreeFellingToAnalytics
  (params (FelledBy :type Actor%20Object%20Reference))
  (bind _returnvalue (Actor|GetOwner FelledBy))
  (if (Networking|IsServer)
    (Utilities|FlowControl|Sequence
      (:then_0
        (Utilities|IsValid _returnvalue
          (:Is%20Valid
            (bind _asfort_player_pawn_athena (Utilities|Casting|CastToFortPlayerPawnAthena _returnvalue)
              (:then
                (bind _asfort_player_state_athena (Utilities|Casting|CastToFortPlayerStateAthena (Variables|/Script/Engine.Pawn|GetPlayerState _asfort_player_pawn_athena)))
                (bind _returnvalue_2 (PlayerInfo|GetDebugStringforUniqueId (PlayerState|GetPlayerUniqueNetID _asfort_player_state_athena)))
                (LocalVariables|SetFeller _returnvalue_2))
              (:CastFailed
                _)))
          (:Is%20Not%20Valid
            _))
        (LocalVariables|SetFeller (Utilities|GetObjectName (Variables|GetFelledBy))))
      (:then_1
        (bind (_returnvalue_3 _success) (CallFunction|GetAnalyticsReporterActor))
        (if _success
          (Class|BPhysicsTreeReporter|ReportLogFelled _returnvalue_3 (Utilities|GetObjectName self) (LocalVariables|GetFeller) (Transformation|GetActorLocation self)))))))

(fn SetOverrideBuoyancy)

(fn ConstructionScript
  (|CallParentFunction))

(fn WhichDistanceSocketIsFurthestAway
  (params (Sockets :type Array%20of%20Names))
  (Utilities|IsValid (Variables|Default|GetCurrentPawn)
    (:Is%20Valid
      (LocalVariables|SetSucceededLocal true)
      (for _array_element Sockets
        (Utilities|Array|AddUnique (LocalVariables|GetAllSocketVectors) (Transformation|GetSocketLocation (Variables|Debug|GetStaticMeshComponent) _array_element)))
      (for _array_element_2 (LocalVariables|GetAllSocketVectors)
        (Utilities|Array|Add (LocalVariables|GetAllDistancesFromPlayer) (Math|Vector|VectorLength (- (Math|Vector|VectorGetAbs _array_element_2) (Math|Vector|VectorGetAbs (Transformation|GetActorLocation (Variables|Default|GetCurrentPawn))))))))
    (:Is%20Not%20Valid
      (LocalVariables|SetSucceededLocal false)))
  (return (Utilities|Array|Get(acopy) (LocalVariables|GetAllSocketVectors) (Math|Float|MaxOfFloatArray (LocalVariables|GetAllDistancesFromPlayer))) (LocalVariables|GetSucceededLocal)))
